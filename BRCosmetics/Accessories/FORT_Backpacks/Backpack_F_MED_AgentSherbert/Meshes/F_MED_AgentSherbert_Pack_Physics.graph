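# Physics asset for the AgentSherbert backpack cosmetic, giving its left and right wings simulated secondary motion.
# A small capsule on the pack bone anchors two mirrored four-segment chains of thin box bodies that shrink toward the tips.
# Each chain is linked segment to segment by constraints with narrow swing and twist limits, matched on both sides, so the wings flex only slightly.
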PhysicsAsset /BRCosmetics/Accessories/FORT_Backpacks/Backpack_F_MED_AgentSherbert/Meshes/F_MED_AgentSherbert_Pack_Physics
bodies {
 dyn_wing_1_l [{
   shapeType=Box
   extentX=10.5
   extentY=2
   extentZ=26
   shapeName=dyn_wing_1_l_box
   center=<5.7818 -0.0031 2.6479>
   rotation=<rot -20.0006 0 -0.0004>
  }]
 dyn_wing_1_r [{
   shapeType=Box
   extentX=10.5
   extentY=2
   extentZ=26
   shapeName=dyn_wing_1_r_box
   center=<-5.8449 0.0031 -2.55>
   rotation=<rot -20.0006 0 -0.0005>
  }]
 dyn_wing_2_l [{
   shapeType=Box
   extentX=8.5
   extentY=2
   extentZ=23.5
   shapeName=dyn_wing_2_l_box
   center=<5.4242 0 -0.3235>
   rotation=<rot -25.0002 0 -0.0001>
  }]
 dyn_wing_2_r [{
   shapeType=Box
   extentX=8.5
   extentY=2
   extentZ=23.5
   shapeName=dyn_wing_2_r_box
   center=<-5.3924 0 0.4047>
   rotation=<rot -25.0002 0 -0.0001>
  }]
 dyn_wing_3_l [{
   shapeType=Box
   extentX=6
   extentY=2
   extentZ=22
   shapeName=dyn_wing_3_l_box
   center=<3.3292 0.0301 -1.3741>
   rotation=<rot -43.7203 0 -0.0006>
  }]
 dyn_wing_3_r [{
   shapeType=Box
   extentX=6
   extentY=2
   extentZ=22
   shapeName=dyn_wing_3_r_box
   center=<-3.1896 -0.0301 1.4404>
   rotation=<rot -43.7203 0 -0.0006>
  }]
 dyn_wing_4_l [{
   shapeType=Box
   extentX=5.5
   extentY=2
   extentZ=17.5
   shapeName=dyn_wing_4_l_box
   center=<2.5106 -0.1514 -2.4563>
   rotation=<rot -55.0005 0 -0.0003>
  }]
 dyn_wing_4_r [{
   shapeType=Box
   extentX=5.5
   extentY=2
   extentZ=17.5
   shapeName=dyn_wing_4_r_box
   center=<-2.3948 0.1514 2.5852>
   rotation=<rot -55.0005 0 -0.0003>
  }]
 pack [{shapeType=Capsule, length=0.505, radius=2.005, rotation=<rot 0 -90 0>}]
}
constraints [
 {bone1Name=dyn_wing_1_r, bone2Name=dyn_wing_2_r, swing1LimitDegrees=20, swing2LimitDegrees=7.5, twistLimitDegrees=5}
 {bone1Name=dyn_wing_2_r, bone2Name=dyn_wing_3_r, swing1LimitDegrees=15, swing2LimitDegrees=5, twistLimitDegrees=3.5}
 {bone1Name=dyn_wing_3_r, bone2Name=dyn_wing_4_r, swing1LimitDegrees=10, swing2LimitDegrees=15, twistLimitDegrees=7.5}
 {bone1Name=dyn_wing_1_l, bone2Name=dyn_wing_2_l, swing1LimitDegrees=20, swing2LimitDegrees=7.5, twistLimitDegrees=5}
 {bone1Name=dyn_wing_2_l, bone2Name=dyn_wing_3_l, swing1LimitDegrees=15, swing2LimitDegrees=5, twistLimitDegrees=3.5}
 {bone1Name=dyn_wing_3_l, bone2Name=dyn_wing_4_l, swing1LimitDegrees=10, swing2LimitDegrees=15, twistLimitDegrees=7.5}
 {bone1Name=pack, bone2Name=dyn_wing_1_l, swing1LimitDegrees=20, swing2LimitDegrees=10, twistLimitDegrees=5}
 {bone1Name=pack, bone2Name=dyn_wing_1_r, swing1LimitDegrees=20, swing2LimitDegrees=10, twistLimitDegrees=5}
]
