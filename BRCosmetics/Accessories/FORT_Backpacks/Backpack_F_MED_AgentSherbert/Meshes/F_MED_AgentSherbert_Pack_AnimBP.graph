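# Animation Blueprint for the AgentSherbert back bling that poses four pairs of wing bones from player movement.
# Update logic eases wing angles and blend alphas toward poses for jump/fall/skydive, dive, parachute, sprint and tactical sprint.
# Otherwise, or with rigid body disabled, phase-offset sine waves flap the wings.
# AnimGraph applies shared left/right angles as additive bone rotations, then layers RigidBody simulation relative to the pack bone.
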AnimBlueprint /BRCosmetics/Accessories/FORT_Backpacks/Backpack_F_MED_AgentSherbert/Meshes/F_MED_AgentSherbert_Pack_AnimBP
target_skeleton=&/F_MED_AgentSherbert_Pack_Skeleton
graphs [
 AnimGraph {nodes [
   n1 = Get alfa_float() -> (alfa_float:double)
   n2 = "Transform (Modify) Bone - Bone: dyn_wing_1_l"!(
    ComponentPose=Local To Component(LocalPose=Input Pose())
    Alpha=@n1
    Translation="0,0,0"
    Rotation_Roll:float=0
    Rotation_Pitch:float=Get wing_1_dive() -> (wing_1_dive:double)
    Rotation_Yaw:float=Get wing_wave_1() -> (wing_wave_1:double)
    Scale="1,1,1"
   ) -> (Pose) {
    rotationMode=BMM_Additive
    rotationSpace=BCS_BoneSpace
    scaleSpace=BCS_BoneSpace
    alphaBoolBlend {blendOption=Linear}
    alphaScaleBiasClamp {bClampResult=true, bInterpResult=true, outRange {min=0.5}}
    boneToModify {boneName=dyn_wing_1_l}
   }
   n3 = "Transform (Modify) Bone - Bone: dyn_wing_2_l"!(
    ComponentPose=@n2
    Alpha=@n1
    Translation="0,0,0"
    Rotation_Roll:float=0
    Rotation_Pitch:float=Get wing_2_dive() -> (wing_2_dive:double)
    Rotation_Yaw:float=Get wing_wave_2() -> (wing_wave_2:double)
    Scale="1,1,1"
   ) -> (Pose) {
    rotationMode=BMM_Additive
    rotationSpace=BCS_BoneSpace
    alphaBoolBlend {blendOption=Linear}
    alphaScaleBiasClamp {bClampResult=true, bInterpResult=true, outRange {min=0.5}}
    boneToModify {boneName=dyn_wing_2_l}
   }
   n4 = "Transform (Modify) Bone - Bone: dyn_wing_3_l"!(
    ComponentPose=@n3
    Alpha=@n1
    Translation="0,0,0"
    Rotation_Roll:float=0
    Rotation_Pitch:float=Get wing_3_dive() -> (wing_3_dive:double)
    Rotation_Yaw:float=Get wing_wave_3() -> (wing_wave_3:double)
    Scale="1,1,1"
   ) -> (Pose) {
    rotationMode=BMM_Additive
    rotationSpace=BCS_BoneSpace
    alphaBoolBlend {blendOption=Linear}
    alphaScaleBiasClamp {bClampResult=true, bInterpResult=true, outRange {min=0.5}}
    boneToModify {boneName=dyn_wing_3_l}
   }
   n5 = "Transform (Modify) Bone - Bone: dyn_wing_4_l"!(
    ComponentPose=@n4
    Alpha=@n1
    Translation="0,0,0"
    Rotation_Roll:float=0
    Rotation_Pitch:float=Get wing_4_dive() -> (wing_4_dive:double)
    Rotation_Yaw:float=Get wing_wave_4() -> (wing_wave_4:double)
    Scale="1,1,1"
   ) -> (Pose) {
    rotationMode=BMM_Additive
    rotationSpace=BCS_BoneSpace
    alphaBoolBlend {blendOption=Linear}
    alphaScaleBiasClamp {bClampResult=true, bInterpResult=true, outRange {min=0.5}}
    boneToModify {boneName=dyn_wing_4_l}
   }
   n6 = "Transform (Modify) Bone - Bone: dyn_wing_1_r"!(
    ComponentPose=@n5
    Alpha=@n1
    Translation="0,0,0"
    Rotation_Roll:float=0
    Rotation_Pitch:float=Get wing_1_dive()
    Rotation_Yaw:float=Get wing_wave_1()
    Scale="1,1,1"
   ) -> (Pose) {
    rotationMode=BMM_Additive
    rotationSpace=BCS_BoneSpace
    scaleSpace=BCS_BoneSpace
    alphaBoolBlend {blendOption=Linear}
    alphaScaleBiasClamp {bClampResult=true, bInterpResult=true, outRange {min=0.5}}
    boneToModify {boneName=dyn_wing_1_r}
   }
   n7 = "Transform (Modify) Bone - Bone: dyn_wing_2_r"!(
    ComponentPose=@n6
    Alpha=@n1
    Translation="0,0,0"
    Rotation_Roll:float=0
    Rotation_Pitch:float=Get wing_2_dive()
    Rotation_Yaw:float=Get wing_wave_2()
    Scale="1,1,1"
   ) -> (Pose) {
    rotationMode=BMM_Additive
    rotationSpace=BCS_BoneSpace
    alphaBoolBlend {blendOption=Linear}
    alphaScaleBiasClamp {bClampResult=true, bInterpResult=true, outRange {min=0.5}}
    boneToModify {boneName=dyn_wing_2_r}
   }
   n8 = "Transform (Modify) Bone - Bone: dyn_wing_3_r"!(
    ComponentPose=@n7
    Alpha=@n1
    Translation="0,0,0"
    Rotation_Roll:float=0
    Rotation_Pitch:float=Get wing_3_dive()
    Rotation_Yaw:float=Get wing_wave_3()
    Scale="1,1,1"
   ) -> (Pose) {
    rotationMode=BMM_Additive
    rotationSpace=BCS_BoneSpace
    alphaBoolBlend {blendOption=Linear}
    alphaScaleBiasClamp {bClampResult=true, bInterpResult=true, outRange {min=0.5}}
    boneToModify {boneName=dyn_wing_3_r}
   }
   n9 = "Transform (Modify) Bone - Bone: dyn_wing_4_r"!(
    ComponentPose=@n8
    Alpha=@n1
    Translation="0,0,0"
    Rotation_Roll:float=0
    Rotation_Pitch:float=Get wing_4_dive()
    Rotation_Yaw:float=Get wing_wave_4()
    Scale="1,1,1"
   ) -> (Pose) {
    rotationMode=BMM_Additive
    rotationSpace=BCS_BoneSpace
    alphaBoolBlend {blendOption=Linear}
    alphaScaleBiasClamp {bClampResult=true, bInterpResult=true, outRange {min=0.5}}
    boneToModify {boneName=dyn_wing_4_r}
   }
   n10 = RigidBody(ComponentPose=@n9, Alpha=Get alfa_float_rigidBody() -> (alfa_float_rigidBody:double)) {
    overridePhysicsAsset=&/F_MED_AgentSherbert_Pack_Physics
    simulationSpace=BaseBoneSpace
    alphaBoolBlend {blendOption=Linear}
    alphaScaleBiasClamp {bClampResult=true, bInterpResult=true}
    baseBoneRef {boneName=pack}
    componentAppliedLinearAccClamp=<1000 1000 1000>
    componentLinearAccScale=<1 1 1>
    componentLinearVelScale=<0.5 0.5 0.5>
    simSpaceSettings {maxAngularAcceleration=500, maxAngularVelocity=500000, maxLinearVelocity=300, worldAlpha=0.5}
   }
   Output Pose(Result=Component To Local(ComponentPose=@n10)) {layerGroup=None}
  ]}
 EventGraph {nodes [
   n1 = Event BlueprintUpdateAnimation()
   > Set Pawn Speed Remap(Pawn Speed Remap:double=MapRangeClamped(Value=Get PawnSpeed2D(), InRangeA=0, InRangeB=549)) -> (Output_Get:double)
   n2 = Get bIsTacticalSprinting()
   n3 = Get bIsFalling()
   n4 = Get bIsJumping()
   n5 = Get bIsSkyDiving()
   > n6 = Branch(Condition=AND Boolean(A=OR Boolean!(A=@n5, B=@n4, C:bool=@n3) -> (ReturnValue), B=NOT Boolean(A=@n2)))
   > n7 = Branch(Condition=Get bRigidBodyEnabled())
   n8 = float * float(A=@n1.DeltaTimeX, B=0.5) {is=title K2Node_PromotableOperator}
   > Set wing_wave_1(wing_wave_1:double=FInterpTo(Current=Get wing_wave_1(), Target=150, DeltaTime=@n8, InterpSpeed=7)) -> (Output_Get:double)
   > Set wing_wave_2(wing_wave_2:double=FInterpTo(Current=Get wing_wave_2(), Target=-120, DeltaTime=@n8, InterpSpeed=7)) -> (Output_Get:double)
   > Set wing_wave_3(wing_wave_3:double=FInterpTo(Current=Get wing_wave_3(), Target=30, DeltaTime=@n8, InterpSpeed=7)) -> (Output_Get:double)
   > Set wing_wave_4(wing_wave_4:double=FInterpTo(Current=Get wing_wave_4(), Target=20, DeltaTime=@n8, InterpSpeed=7)) -> (Output_Get:double)
   > Set wing_1_dive(wing_1_dive:double=0) -> (Output_Get:double)
   > Set wing_2_dive(wing_2_dive:double=0) -> (Output_Get:double)
   > Set wing_3_dive(wing_3_dive:double=0) -> (Output_Get:double)
   > Set wing_4_dive(wing_4_dive:double=0) -> (Output_Get:double)
   > Set alfa_float(alfa_float:double=0.45) -> (Output_Get:double)
   > n9 = Branch(Condition=Get bIsSkydiveDiveMode())
   > Set alfa_float_rigidBody(alfa_float_rigidBody:double=0.5) -> (Output_Get:double)
   > Set wing_wave_1(wing_wave_1=FInterpTo(Current=Get wing_wave_1(), Target=-20, DeltaTime=@n1.DeltaTimeX, InterpSpeed=7.5))
   > Set wing_wave_2(wing_wave_2=FInterpTo(Current=Get wing_wave_2(), Target=-90, DeltaTime=@n1.DeltaTimeX, InterpSpeed=7.5))
   > Set wing_wave_3(wing_wave_3=FInterpTo(Current=Get wing_wave_3(), Target=-15, DeltaTime=@n1.DeltaTimeX, InterpSpeed=7.5))
   > Set wing_wave_4(wing_wave_4=FInterpTo(Current=Get wing_wave_4(), Target=-5, DeltaTime=@n1.DeltaTimeX, InterpSpeed=5))
   n10 = float * float(A=@n1.DeltaTimeX, B=10)
   > Set wing_1_dive(wing_1_dive=FInterpTo(Current=Get wing_1_dive(), Target=0, DeltaTime=@n10, InterpSpeed=1))
   > Set wing_2_dive(wing_2_dive=FInterpTo(Current=Get wing_2_dive(), Target=-30, DeltaTime=@n10, InterpSpeed=1))
   > Set wing_3_dive(wing_3_dive=FInterpTo(Current=Get wing_3_dive(), Target=-120, DeltaTime=@n10, InterpSpeed=1))
   > Set wing_4_dive(wing_4_dive=FInterpTo(Current=Get wing_4_dive(), Target=-20, DeltaTime=@n10, InterpSpeed=1))
   > Set alfa_float(alfa_float=1)
   Branch(execute=@n9.else, Condition=Get bIsParachuteOpened())
   > Set wing_wave_1(wing_wave_1=FInterpTo(Current=Get wing_wave_1(), Target=60, DeltaTime=@n10, InterpSpeed=1.5))
   > Set wing_wave_2(wing_wave_2=FInterpTo(Current=Get wing_wave_2(), Target=-20, DeltaTime=@n10, InterpSpeed=0.5))
   > Set wing_wave_3(wing_wave_3=FInterpTo(Current=Get wing_wave_3(), Target=0, DeltaTime=@n10, InterpSpeed=0.5))
   > Set wing_wave_4(wing_wave_4=FInterpTo(Current=Get wing_wave_4(), Target=0, DeltaTime=@n10, InterpSpeed=1))
   > Set wing_1_dive(wing_1_dive=20)
   > Set wing_2_dive(wing_2_dive=0)
   > Set wing_3_dive(wing_3_dive=0)
   > Set wing_4_dive(wing_4_dive=0)
   > Set alfa_float(alfa_float=1)
   > Set alfa_float_rigidBody(alfa_float_rigidBody=0.45)
   n11 = float + float(A=float * float(A=GetGameTimeInSeconds(), B=float + float(B=1.5)), B=20)
   n12 = "Sin (Radians)"(A=@n11)
   Set wing_wave_1(execute=[@n7.else, @n16.else, @n17.else], wing_wave_1=float + float(A=float * float(A=@n12, B=30), B=10))
   n13 = "Sin (Radians)"(A=float + float(A=@n11, B=20))
   > Set wing_wave_2(wing_wave_2=float + float(A=float * float(A=@n13, B=25), B=10))
   n14 = "Sin (Radians)"(A=float + float(A=@n11, B=40))
   > Set wing_wave_3(wing_wave_3=float + float(A=float * float(A=@n14, B=40), B=10))
   n15 = "Sin (Radians)"(A=float + float(A=@n11, B=60))
   > Set wing_wave_4(wing_wave_4=float + float(A=float * float(A=@n15, B=15), B=10))
   > Set wing_1_dive(wing_1_dive=0)
   > Set wing_2_dive(wing_2_dive=0)
   > Set wing_3_dive(wing_3_dive=0)
   > Set wing_4_dive(wing_4_dive=0)
   > Set alfa_float(alfa_float=0.5)
   > Set alfa_float_rigidBody(alfa_float_rigidBody=0.45)
   n16 = Branch(execute=@n6.else, Condition=Get bIsSprinting())
   > n17 = Branch(Condition=Get bRigidBodyEnabled())
   > Set wing_wave_1(wing_wave_1=FInterpTo(Current=Get wing_wave_1(), Target=-120, DeltaTime=@n1.DeltaTimeX, InterpSpeed=0.25))
   > Set wing_wave_2(wing_wave_2=FInterpTo(Current=Get wing_wave_2(), Target=-170, DeltaTime=@n1.DeltaTimeX, InterpSpeed=0.25))
   > Set wing_wave_3(wing_wave_3=FInterpTo(Current=Get wing_wave_3(), Target=-120, DeltaTime=@n1.DeltaTimeX, InterpSpeed=0.25))
   > Set wing_wave_4(wing_wave_4=FInterpTo(Current=Get wing_wave_4(), Target=-60, DeltaTime=@n1.DeltaTimeX, InterpSpeed=0.25))
   > Set alfa_float_rigidBody(alfa_float_rigidBody=1)
   > Set alfa_float(alfa_float=1)
   > Branch(Condition=Get bIsTacticalSprinting())
   > Set alfa_float_rigidBody(alfa_float_rigidBody=0.9)
   > Set wing_wave_1(wing_wave_1=FInterpTo(Current=Get wing_wave_1(), Target=-150, DeltaTime=@n1.DeltaTimeX, InterpSpeed=100))
   > Set wing_wave_2(wing_wave_2=FInterpTo(Current=Get wing_wave_2(), Target=-150, DeltaTime=@n1.DeltaTimeX, InterpSpeed=100))
   > Set wing_wave_3(wing_wave_3=FInterpTo(Current=Get wing_wave_3(), Target=30, DeltaTime=@n1.DeltaTimeX, InterpSpeed=100))
   > Set wing_wave_4(wing_wave_4=FInterpTo(Current=Get wing_wave_4(), Target=-20, DeltaTime=@n1.DeltaTimeX, InterpSpeed=100))
   > Set wing_1_dive(wing_1_dive=0)
   > Set wing_2_dive(wing_2_dive=0)
   > Set wing_3_dive(wing_3_dive=0)
   > Set wing_4_dive(wing_4_dive=0)
   > Set alfa_float(alfa_float=0.8)
  ]}
]
variables [
 alfa_float {type=number, default=1}
 alfa_float_rigidBody {type=number, default=0.5}
 Pawn Speed Remap {type=number}
 Sim Space Settings {
  type=SimSpaceSettings
  default {
   dampingAlpha=1
   maxAngularAcceleration=500
   maxAngularVelocity=500000
   maxLinearAcceleration=10000
   maxLinearVelocity=300
   velocityScaleZ=1
   worldAlpha=0.5
   externalLinearDragV=<0 0 0>
  }
 }
 wing_1_dive {type=number}
 wing_2_dive {type=number}
 wing_3_dive {type=number}
 wing_4_dive {type=number}
 wing_wave_1 {type=number}
 wing_wave_2 {type=number}
 wing_wave_3 {type=number}
 wing_wave_4 {type=number}
]
